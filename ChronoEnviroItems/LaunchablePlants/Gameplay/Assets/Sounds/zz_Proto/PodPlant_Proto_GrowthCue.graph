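# Prototype sound cue for the growth sound of a launchable pod plant in the ChronoEnviroItems feature.
# A random node picks among three PodPlant_Grow waves with equal weight, and a modulator node at the root passes the result on.
# It plays in the gadgets and items sound class, with reality pod prop attenuation set alongside an enabled attenuation override.
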SoundCue /ChronoEnviroItems/LaunchablePlants/Gameplay/Assets/Sounds/zz_Proto/PodPlant_Proto_GrowthCue
root=Modulator_0
nodes [
 {id=Modulator_0, children [Random_0]}
 {id=Random_0, children [WavePlayer_0,WavePlayer_1,WavePlayer_2], properties {weights [1,1,1]}}
 {id=WavePlayer_0, properties {soundWaveAssetPtr=&/PodPlant_Grow_3}}
 {id=WavePlayer_1, properties {soundWaveAssetPtr=&/PodPlant_Grow_1}}
 {id=WavePlayer_2, properties {soundWaveAssetPtr=&/PodPlant_Grow_2}}
]
properties {
 attenuationSettings=/Game/Sounds/Attenuation/Fort_Prop_RealityPods_Att
 bOverrideAttenuation=true
 maxDistance=2100
 soundClassObject=&Fort_Gadgets_Items
 volumeMultiplier=1
 attenuationOverrides {falloffDistance=2000, attenuationShapeExtents {x=100}}
}
